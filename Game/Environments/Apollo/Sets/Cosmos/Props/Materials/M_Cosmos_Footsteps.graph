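# Translucent lit material for footprints on the Cosmos set, with runtime virtual texture support.
# Base color is vertex color scaled by a parameter, and opacity multiplies vertex alpha by the footprint texture's red channel and a scalar.
# A scaled normal map, with tangent-space normals disabled, joins specular and roughness parameters in one attribute set.
# That set feeds a virtual texture output and a feature switch through EnableRVT, then a shading path switch.
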Material /Game/Environments/Apollo/Sets/Cosmos/Props/Materials/M_Cosmos_Footsteps
graph {
 properties {
  bAutomaticallySetUsageInEditor=false
  bCastRayTracedShadows=false
  bTangentSpaceNormal=false
  bUseMaterialAttributes=true
  blendMode=BLEND_Translucent
  pixelDepthOffsetMode=PDOM_Legacy
  usedShadingModels=Default Lit
 }
 nodes [
  Constant3Vector() {constant=<rgba 0.3372 0.2384 0.1356 0>}
  n1 = VertexColor()
  n2 = TextureSample() {texture=/Game/Environments/Apollo/Sets/Cosmos/Props/Textures/T_Cosmos_Footprints_D}
  n3 = ScalarParameter() {defaultValue=0.65, parameterName=gg4e3}
  n4 = SetMaterialAttributes!(
   MaterialAttributes
   Base Color=Multiply(A=@n1, B=ScalarParameter() {defaultValue=0.35, parameterName=ColorMod})
   Specular=ScalarParameter() {defaultValue=0.35, parameterName=spec}
   Roughness=ScalarParameter() {defaultValue=0.8, parameterName=rough}
   Normal=Multiply(
    A=TextureSample() {samplerType=SAMPLERTYPE_Normal, texture=/Game/Environments/Apollo/Sets/Cosmos/Props/Textures/T_Cosmos_Footprints_N}.RGB
    B=VectorParameter() {parameterName=NormalMod, defaultValue=<rgba 2 2 1 1>}.RGB
   )
   Opacity=Multiply(A=@n1.A, B=Multiply(A=@n2.R, B=@n3))
  ) {attributeSetTypes [
    #1
    #2
    #3
    #4
    #5
   ]}
  n5 = GetMaterialAttributes(None=@n4) {attributeGetTypes [
    #1
    #2
    #3
    #4
    #5
   ]}
  RuntimeVirtualTextureOutput!(
   BaseColor=@n5.Base Color
   Specular=@n5.Base Color
   Roughness=@n5.Base Color
   Normal=@n5.Base Color
   WorldHeight
   Opacity=@n5.Base Color
   Mask
   Displacement
   Mask4
  )
  n6 = VirtualTextureFeatureSwitch!(
   No=@n4
   Yes=MaterialFunctionCall!(RVT=@n5.MaterialAttributes, ByPass RVT=@n4) {materialFunction=/Game/Athena/Apollo/Maps/Landscape/VT/MaterialFunctions/EnableRVT}.Result
  )
  ShadingPathSwitch(Default=@n6, Deferred=@n6, Forward=@n6, Mobile=@n6)
  VectorParameter() {parameterName=f433, defaultValue=<rgba 0.2 0.1461 0.081 1>}
 ]
}
